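# Physics asset for the back face-accessory part of the female medium Rebirth Medic body, covering its dynamic hair bones.
# Four chains of box bodies (mid-left, back-left, back-middle, back-right) hang from the head through limited swing and twist constraints, with the right chain generally looser, especially in twist.
# Capsules on head, upper spine, both clavicles and only the right upper arm supply body shapes, likely as hair colliders; the clavicle-to-spine constraints leave rotation free.
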PhysicsAsset /BRCosmetics/Characters/Player/Female/Medium/Bodies/F_MED_Rebirth_Medic/Meshes/Parts/F_MED_Rebirth_Medic_FaceAcc_Back_Physics
bodies {
 clavicle_l [{shapeType=Capsule, length=7.5, radius=6.7226, center=<5.4735 -1.5329 1.7604>, rotation=<rot 90 0 0>}]
 clavicle_r [{shapeType=Capsule, length=8.5, radius=5.7226, center=<-6.6711 -0.0001 -1.0393>, rotation=<rot 90 0 0>}]
 dyn_hair_bk_02 [{shapeType=Box, extentX=9.8766, extentY=4.9314, extentZ=2.2208, center=<4.4263 0.6031 -0.0003>, rotation=<rot 0.0001 7.7602 11.569>}]
 dyn_hair_bk_l_01 [{shapeType=Box, extentX=8.8766, extentY=5.4314, extentZ=1.7208, center=<4.1457 0 0>}]
 dyn_hair_bk_l_02 [{shapeType=Box, extentX=6.8766, extentY=5.4314, extentZ=1.7208, center=<3.4338 0 -0.0002>}]
 dyn_hair_bk_l_parent [{shapeType=Box, extentX=3.8766, extentY=5.4314, extentZ=1.2208, center=<1.8725 -0.0001 0>}]
 dyn_hair_bk_mid_01 [{shapeType=Box, extentX=8.8766, extentY=5.4314, extentZ=1.7208, center=<4.6521 -0.0968 -0.0002>, rotation=<rot 0 -2.2578 -12.6302>}]
 dyn_hair_bk_mid_parent [{shapeType=Box, extentX=4.8766, extentY=3.9314, extentZ=1.2208, center=<2.296 0.0125 0>, rotation=<rot 0 7.025 15.0536>}]
 dyn_hair_bk_r_01 [{shapeType=Box, extentX=10.8766, extentY=3.4314, extentZ=0.7208, center=<5.3418 0 0>}]
 dyn_hair_bk_r_02 [{shapeType=Box, extentX=7.3766, extentY=2.4314, extentZ=0.7208, center=<3.1675 0 0>}]
 dyn_hair_bk_r_parent [{shapeType=Box, extentX=4.8766, extentY=3.4314, extentZ=0.7208, center=<2.3278 -0.2863 -0.003>, rotation=<rot -3.4435 -8.3485 -0.0001>}]
 dyn_hair_mid_l_01 [{shapeType=Box, extentX=1.8766, extentY=3.4314, extentZ=1.2208, center=<0.9159 0 0>}]
 dyn_hair_mid_l_parent [{shapeType=Box, extentX=5.3766, extentY=3.4314, extentZ=1.2208, center=<2.1949 0 0>}]
 head [{shapeType=Capsule, length=9, radius=5.5, center=<-2.2996 -0.3476 -0.0232>, rotation=<rot 89.4688 -179.9995 165.2721>}]
 spine_05 [{shapeType=Capsule, length=4, radius=6.5, center=<8.3714 -0.3912 0.0167>, rotation=<rot -90 0 15.0009>}]
 upperarm_r [{shapeType=Capsule, length=11.5, radius=5.25, center=<2.7087 1.8577 8.56>, rotation=<rot -52.7633 -166.0862 138.3983>}]
}
constraints [
 {bone1Name=dyn_hair_mid_l_parent, bone2Name=head, swing1LimitDegrees=30, swing2LimitDegrees=30, twistLimitDegrees=3}
 {bone1Name=dyn_hair_bk_l_parent, bone2Name=head, swing1LimitDegrees=30, swing2LimitDegrees=30, twistLimitDegrees=3}
 {bone1Name=dyn_hair_mid_l_01, bone2Name=dyn_hair_mid_l_parent, swing1LimitDegrees=60, swing2LimitDegrees=60, twistLimitDegrees=3}
 {bone1Name=dyn_hair_bk_l_01, bone2Name=dyn_hair_bk_l_parent, swing1LimitDegrees=30, swing2LimitDegrees=30, twistLimitDegrees=3}
 {bone1Name=dyn_hair_bk_l_02, bone2Name=dyn_hair_bk_l_01, swing1LimitDegrees=30, swing2LimitDegrees=30, twistLimitDegrees=3}
 {bone1Name=dyn_hair_bk_mid_parent, bone2Name=head, swing1LimitDegrees=30, swing2LimitDegrees=30, twistLimitDegrees=3}
 {bone1Name=dyn_hair_bk_mid_01, bone2Name=dyn_hair_bk_mid_parent, swing1LimitDegrees=30, swing2LimitDegrees=30, twistLimitDegrees=3}
 {bone1Name=dyn_hair_bk_02, bone2Name=dyn_hair_bk_mid_01, swing1LimitDegrees=30, swing2LimitDegrees=30, twistLimitDegrees=3}
 {bone1Name=dyn_hair_bk_r_parent, bone2Name=head, swing1LimitDegrees=55, swing2LimitDegrees=55, twistLimitDegrees=12}
 {bone1Name=dyn_hair_bk_r_01, bone2Name=dyn_hair_bk_r_parent, swing1LimitDegrees=55, swing2LimitDegrees=55, twistLimitDegrees=12}
 {bone1Name=dyn_hair_bk_r_02, bone2Name=dyn_hair_bk_r_01, swing1LimitDegrees=45, swing2LimitDegrees=45, twistLimitDegrees=10}
 {
  bone1Name=clavicle_l
  bone2Name=spine_05
  swing1LimitDegrees=45
  swing1Motion=Free
  swing2LimitDegrees=45
  swing2Motion=Free
  twistLimitDegrees=45
  twistMotion=Free
 }
 {
  bone1Name=clavicle_r
  bone2Name=spine_05
  swing1LimitDegrees=45
  swing1Motion=Free
  swing2LimitDegrees=45
  swing2Motion=Free
  twistLimitDegrees=45
  twistMotion=Free
 }
]
