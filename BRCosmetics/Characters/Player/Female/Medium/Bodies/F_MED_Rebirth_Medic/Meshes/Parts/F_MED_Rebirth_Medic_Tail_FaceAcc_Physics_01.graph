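# Physics asset for the tail hair part of the female Rebirth Medic cosmetic, letting the hair swing and collide with the body.
# Four capsule hair bodies form a constrained chain with narrow twist and wider swing limits, rooted on a small driver body.
# The driver joins to spine_05 with free swing and twist.
# Unconstrained capsules on head, neck, upper spine and left forearm supply collision shapes.
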PhysicsAsset /BRCosmetics/Characters/Player/Female/Medium/Bodies/F_MED_Rebirth_Medic/Meshes/Parts/F_MED_Rebirth_Medic_Tail_FaceAcc_Physics_01
bodies {
 dyn_driver_main_hair [{shapeType=Capsule, length=0.505, radius=0.505, rotation=<rot 0 -90 0>}]
 dyn_main_hair_01 [{shapeType=Capsule, length=6.0608, radius=3.1882, center=<1.7846 0.126 0.1097>, rotation=<rot 0 90 -90>}]
 dyn_main_hair_02 [{shapeType=Capsule, length=3, radius=2.3225, center=<2.366 0 0>, rotation=<rot 0 90 -90>}]
 dyn_main_hair_03 [{shapeType=Capsule, length=1.8561, radius=2.1851, center=<1.8675 -0.006 -0.064>, rotation=<rot 0 90 -90>}]
 dyn_main_hair_04 [{shapeType=Capsule, length=1.847, radius=1.4771, center=<2.1939 -0.5336 -0.0102>, rotation=<rot 0 90 -90>}]
 head [{shapeType=Capsule, length=5, radius=6.75, rotation=<rot 90 -0.0003 -0.0002>}]
 lowerarm_l [{shapeType=Capsule, length=25, radius=6.75, center=<-11.6871 5.577 1.3776>, rotation=<rot 77.1264 36.7567 65.5791>}]
 neck_01 [{shapeType=Capsule, length=6.25, radius=4.5, center=<0.3642 1.1246 0.0048>, rotation=<rot 89.3571 -179.9998 -162.0508>}]
 spine_04 [{shapeType=Capsule, length=12.25, radius=17.75, center=<-6.5908 -5.4418 -0.9616>, rotation=<rot 89.3251 179.9999 180>}]
 spine_05 [{shapeType=Capsule, length=15, radius=8.25, center=<2.3703 -1.851 0.0001>}]
}
constraints [
 {
  bone1Name=dyn_driver_main_hair
  bone2Name=spine_05
  swing1LimitDegrees=45
  swing1Motion=Free
  swing2LimitDegrees=45
  swing2Motion=Free
  twistLimitDegrees=45
  twistMotion=Free
 }
 {bone1Name=dyn_main_hair_01, bone2Name=dyn_driver_main_hair, swing1LimitDegrees=30, swing2LimitDegrees=25, twistLimitDegrees=5}
 {bone1Name=dyn_main_hair_02, bone2Name=dyn_main_hair_01, swing1LimitDegrees=30, swing2LimitDegrees=25, twistLimitDegrees=5}
 {bone1Name=dyn_main_hair_03, bone2Name=dyn_main_hair_02, swing1LimitDegrees=30, swing2LimitDegrees=25, twistLimitDegrees=5}
 {bone1Name=dyn_main_hair_04, bone2Name=dyn_main_hair_03, swing1LimitDegrees=30, swing2LimitDegrees=25, twistLimitDegrees=5}
]
