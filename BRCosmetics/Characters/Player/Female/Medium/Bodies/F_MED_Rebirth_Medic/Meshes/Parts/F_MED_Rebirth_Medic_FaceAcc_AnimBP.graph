AnimBlueprint /BRCosmetics/Characters/Player/Female/Medium/Bodies/F_MED_Rebirth_Medic/Meshes/Parts/F_MED_Rebirth_Medic_FaceAcc_AnimBP
target_skeleton=&/F_MED_Rebirth_Medic_FaceAcc_Skeleton
graphs [
 AnimGraph {nodes [
   n1 = Get head_rot_back() -> (head_rot_back:double)
   "Transform (Modify) Bone - Bone: dyn_hair_bk_r_02"(Alpha=@n1, Translation="0,0,0", Rotation="0,0,0", Scale="0.5,0.5,0.5") {
    rotationMode=BMM_Additive
    rotationSpace=BCS_BoneSpace
    scaleMode=BMM_Additive
    scaleSpace=BCS_BoneSpace
    alphaBoolBlend {blendOption=Linear}
    boneToModify {boneName=dyn_hair_bk_r_02}
    scale=<0.5 0.5 0.5>
   }
   n2 = "Transform (Modify) Bone - Bone: dyn_hair_bk_l_01"(
    ComponentPose=Local To Component(LocalPose=Use cached pose 'PoseCache'())
    Alpha=1
    Translation="2.239,-1.363,1.7"
    Rotation="0,0,0"
    Scale="1,1,1"
   ) {
    rotationMode=BMM_Additive
    rotationSpace=BCS_BoneSpace
    translationMode=BMM_Additive
    translationSpace=BCS_BoneSpace
    alphaBoolBlend {blendOption=Linear}
    boneToModify {boneName=dyn_hair_bk_l_01}
    translation=<2.239 -1.363 1.7>
   }
   n3 = "Transform (Modify) Bone - Bone: dyn_hair_bk_mid_01"(ComponentPose=@n2, Alpha=1, Translation="1.5,-1,-0.5", Rotation="0,0,0", Scale="1,1,1") {
    rotationMode=BMM_Additive
    rotationSpace=BCS_BoneSpace
    translationMode=BMM_Additive
    translationSpace=BCS_BoneSpace
    alphaBoolBlend {blendOption=Linear}
    boneToModify {boneName=dyn_hair_bk_mid_01}
    translation=<1.5 -1 -0.5>
   }
   n4 = "Transform (Modify) Bone - Bone: dyn_hair_bk_l_01"(
    ComponentPose=Local To Component(LocalPose=Use cached pose 'PoseCache'())
    Alpha=1
    Translation="-0.769,0.470,1"
    Rotation="0,0,0"
    Scale="1,1,1"
   ) {
    rotationMode=BMM_Additive
    rotationSpace=BCS_BoneSpace
    translationMode=BMM_Additive
    translationSpace=BCS_BoneSpace
    alphaBoolBlend {blendOption=Linear}
    boneToModify {boneName=dyn_hair_bk_l_01}
    translation=<-0.769 0.47 1>
   }
   n5 = "Transform (Modify) Bone - Bone: dyn_hair_bk_mid_01"(ComponentPose=@n4, Alpha=1, Translation="-0.940,0.409,1", Rotation="0,0,0", Scale="1,1,1") {
    rotationMode=BMM_Additive
    rotationSpace=BCS_BoneSpace
    translationMode=BMM_Additive
    translationSpace=BCS_BoneSpace
    alphaBoolBlend {blendOption=Linear}
    boneToModify {boneName=dyn_hair_bk_mid_01}
    translation=<-0.94 0.409 1>
   }
   Output Pose(Result=Blend Poses by bool(
     BlendPose_0=Component To Local(ComponentPose=@n3)
     BlendPose_1=Component To Local(ComponentPose=@n5)
     bActiveValue=Get hair_base_active() -> (hair_base_active:bool)
    ) {blendType=Linear}) {layerGroup=None}
   n6 = Get head_rot_back()
   n7 = Look At - Bone: dyn_hair_bk_l_base(ComponentPose=Local To Component(LocalPose=Copy Pose From Mesh() {bUseAttachedParent=true}), Alpha=@n6) {
    lookAtClamp=100
    alphaBoolBlend {blendOption=Linear}
    boneToModify {boneName=dyn_hair_bk_l_base}
    lookAtLocation=<1 1 5>
    lookAtTarget {boneReference {boneName=spine_04}}
    lookAt_Axis {axis=<1 0>}
    lookUp_Axis {axis {x=0, z=1}}
   }
   n8 = Look At - Bone: dyn_hair_bk_l_01(ComponentPose=@n7, Alpha=@n6) {
    lookAtClamp=100
    alphaBoolBlend {blendOption=Linear}
    boneToModify {boneName=dyn_hair_bk_l_01}
    lookAtLocation=<1 1 5>
    lookAtTarget {boneReference {boneName=spine_04}}
    lookAt_Axis {axis=<1 0>}
    lookUp_Axis {axis {x=0, z=1}}
   }
   n9 = Look At - Bone: dyn_hair_bk_l_02(ComponentPose=@n8, Alpha=@n6) {
    lookAtClamp=100
    alphaBoolBlend {blendOption=Linear}
    boneToModify {boneName=dyn_hair_bk_l_02}
    lookAtLocation=<1 1 5>
    lookAtTarget {boneReference {boneName=spine_03}}
    lookAt_Axis {axis=<1 0>}
    lookUp_Axis {axis {x=0, z=1}}
   }
   n10 = Look At - Bone: dyn_hair_bk_mid_base(ComponentPose=@n9, Alpha=@n6) {
    lookAtClamp=100
    alphaBoolBlend {blendOption=Linear}
    boneToModify {boneName=dyn_hair_bk_mid_base}
    lookAtLocation=<1 1 5>
    lookAtTarget {boneReference {boneName=spine_04}}
    lookAt_Axis {axis=<1 0>}
    lookUp_Axis {axis {x=0, z=1}}
   }
   n11 = Look At - Bone: dyn_hair_bk_mid_01(ComponentPose=@n10, Alpha=@n6) {
    lookAtClamp=100
    alphaBoolBlend {blendOption=Linear}
    boneToModify {boneName=dyn_hair_bk_mid_01}
    lookAtLocation=<1 1 5>
    lookAtTarget {boneReference {boneName=spine_04}}
    lookAt_Axis {axis=<1 0>}
    lookUp_Axis {axis {x=0, z=1}}
   }
   n12 = Look At - Bone: dyn_hair_bk_02(ComponentPose=@n11, Alpha=@n6) {
    lookAtClamp=100
    alphaBoolBlend {blendOption=Linear}
    boneToModify {boneName=dyn_hair_bk_02}
    lookAtLocation=<1 1 5>
    lookAtTarget {boneReference {boneName=spine_04}}
    lookAt_Axis {axis=<1 0>}
    lookUp_Axis {axis {x=0, z=1}}
   }
   n13 = Look At - Bone: dyn_hair_base(ComponentPose=@n12, Alpha=Get headSwingDelta() -> (headSwingDelta:double)) {
    lookAtClamp=100
    alphaBoolBlend {blendOption=Linear}
    boneToModify {boneName=dyn_hair_base}
    lookAtLocation=<1 1 5>
    lookAtTarget {boneReference {boneName=lowerarm_l}}
    lookAt_Axis {axis=<1 0>}
    lookUp_Axis {axis {z=1}}
   }
   n14 = RigidBody(ComponentPose=@n13, Alpha=Get Skydiving_Back_Alpha() -> (Skydiving_Back_Alpha:double)) {
    bClampLinearTranslationLimitToRefPose=true
    bOverrideWorldGravity=true
    overridePhysicsAsset=&/F_MED_Rebirth_Medic_FaceAcc_Back_Physics
    simulationSpace=BaseBoneSpace
    alphaBoolBlend {blendOption=Linear}
    baseBoneRef {boneName=spine_05}
    componentAppliedLinearAccClamp=<1000 1000 1000>
    componentLinearAccScale=<0.25 0.25 0.25>
    componentLinearVelScale=<0.25 0.25 0.25>
    overrideWorldGravity {z=-980}
   }
   n15 = RigidBody(ComponentPose=@n14, Alpha=Get Skydiving_Front_Alpha() -> (Skydiving_Front_Alpha:double)) {
    bClampLinearTranslationLimitToRefPose=true
    bOverrideWorldGravity=true
    overridePhysicsAsset=&/F_MED_Rebirth_Medic_Frnt_FaceAcc_Physics_02
    simulationSpace=BaseBoneSpace
    alphaBoolBlend {blendOption=Linear}
    baseBoneRef {boneName=neck_01}
    componentAppliedLinearAccClamp=<1000 1000 1000>
    componentLinearAccScale=<0.5 0.5 0.25>
    componentLinearVelScale=<0.5 0.5 0.25>
    overrideWorldGravity {z=-980}
   }
   n16 = RigidBody(ComponentPose=@n15, Alpha=0.7) {
    bClampLinearTranslationLimitToRefPose=true
    bOverrideWorldGravity=true
    overridePhysicsAsset=&/F_MED_Rebirth_Medic_Sides_FaceAcc_Physics
    simulationSpace=BaseBoneSpace
    alphaBoolBlend {blendOption=Linear}
    baseBoneRef {boneName=root}
    componentAppliedLinearAccClamp=<300 300 1050>
    componentLinearAccScale=<0.2 0.2 0.7>
    componentLinearVelScale=<0.2 0.2 0.7>
    overrideWorldGravity {z=-980}
   }
   n17 = RigidBody(ComponentPose=@n16) {
    bClampLinearTranslationLimitToRefPose=true
    bOverrideWorldGravity=true
    overridePhysicsAsset=&/F_MED_Rebirth_Medic_Tail_FaceAcc_Physics_01
    simulationSpace=BaseBoneSpace
    alphaBoolBlend {blendOption=Linear}
    baseBoneRef {boneName=root}
    componentAppliedLinearAccClamp=<600 600 600>
    componentLinearAccScale=<0.5 0.5 0.35>
    componentLinearVelScale=<0.5 0.5 0.35>
    overrideWorldGravity {z=-980}
   }
   n18 = Get head_rot_back()
   n19 = "Transform (Modify) Bone - Bone: dyn_hair_bk_l_01"(ComponentPose=@n17, Alpha=@n18, Translation="0,0,0", Rotation="10,20,0", Scale="0.8,0.8,0.8") {
    rotationMode=BMM_Additive
    rotationSpace=BCS_BoneSpace
    scaleMode=BMM_Additive
    scaleSpace=BCS_BoneSpace
    alphaBoolBlend {blendOption=Linear}
    boneToModify {boneName=dyn_hair_bk_l_01}
    rotation {pitch=10, yaw=20}
    scale=<0.8 0.8 0.8>
   }
   n20 = "Transform (Modify) Bone - Bone: dyn_hair_bk_mid_01"(ComponentPose=@n19, Alpha=@n18, Translation="0,0,0", Rotation="30,20,0", Scale="0.8,0.8,0.8") {
    rotationMode=BMM_Additive
    rotationSpace=BCS_BoneSpace
    scaleMode=BMM_Additive
    scaleSpace=BCS_BoneSpace
    alphaBoolBlend {blendOption=Linear}
    boneToModify {boneName=dyn_hair_bk_mid_01}
    rotation {pitch=30, yaw=20}
    scale=<0.8 0.8 0.8>
   }
   n21 = "Transform (Modify) Bone - Bone: dyn_hair_bk_r_01"(ComponentPose=@n20, Alpha=@n1, Translation="0,0,0", Rotation="0,0,0", Scale="0.5,0.5,0.5") {
    rotationMode=BMM_Additive
    rotationSpace=BCS_BoneSpace
    scaleMode=BMM_Additive
    scaleSpace=BCS_BoneSpace
    alphaBoolBlend {blendOption=Linear}
    boneToModify {boneName=dyn_hair_bk_r_01}
    scale=<0.5 0.5 0.5>
   }
   Save cached pose 'PoseCache'(Pose=Component To Local(ComponentPose=@n21)) {cacheName=PoseCache}
  ]}
 EventGraph {nodes [
   Event BlueprintUpdateAnimation()
   > Get!(execute:Exec) -> (then:Exec, else:Exec, Mesh_To_Copy:Skeletal Mesh Component Object Reference)
   > n1 = Sequence()
   n2 = Get Mesh_To_Copy() -> (Mesh_To_Copy:Skeletal Mesh Component Object Reference)
   n3 = BreakTransform(InTransform=GetDeltaTransformFromRefPose(self=@n2, BoneName=head, BaseName=spine_05))
   Set headSwingDelta(execute=@n1.then_1, headSwingDelta:double=MapRangeClamped(Value=BreakRotator(InRot=@n3.Rotation).Yaw, InRangeA=17, InRangeB=-44)) -> (Output_Get:double)
   n4 = BreakRotator(InRot=@n3.Rotation)
   > Set headTwist_r(headTwist_r:double=MapRangeClamped(Value=@n4, InRangeA=15, InRangeB=-37)) -> (Output_Get:double)
   > Set headTwist_l(headTwist_l:double=MapRangeClamped(Value=@n4, InRangeA=15, InRangeB=65)) -> (Output_Get:double)
   n5 = BreakTransform(InTransform=GetDeltaTransformFromRefPose(self=@n2, BoneName=upperarm_r, BaseName=spine_05))
   > Set armSwingDelta_r(armSwingDelta_r:double=MapRangeClamped(Value=BreakRotator(InRot=@n5.Rotation).Yaw, InRangeA=45, InRangeB=130)) -> (Output_Get:double)
   n6 = BreakTransform(InTransform=GetDeltaTransformFromRefPose(self=@n2, BoneName=upperarm_l, BaseName=spine_05))
   > Set armSwingDelta_l(armSwingDelta_l:double=MapRangeClamped(Value=BreakRotator(InRot=@n6.Rotation).Yaw, InRangeA=12, InRangeB=100)) -> (Output_Get:double)
   n7 = BreakTransform(InTransform=GetDeltaTransformFromRefPose(self=@n2, BoneName=head, BaseName=spine_05))
   > Set head_rot_back(head_rot_back:double=MapRangeClamped(Value=BreakRotator(InRot=@n7.Rotation).Yaw, InRangeA=0, InRangeB=-48)) -> (Output_Get:double)
   n8 = Get bIsSkyDiving()
   n9 = Get bIsSkydiveFloating()
   n10 = Get bIsSkydiveFloating()
   Set Skydiving_Front_Alpha(execute=@n1.then_0, Skydiving_Front_Alpha:double=SelectFloat(A=0.35, B=1, bPickA=OR Boolean!(A=@n8, B=@n9, C:bool=@n10) -> (ReturnValue))) -> (Output_Get:double)
   n11 = Get bIsSkyDiving()
   n12 = Get bIsSkydiveFloating()
   n13 = Get bIsSkydiveFloating()
   > Set Skydiving_Back_Alpha(Skydiving_Back_Alpha:double=SelectFloat(A=0.85, B=1, bPickA=OR Boolean(A=@n11, B=@n12, C=@n13))) -> (Output_Get:double)
   Event OnCharacterCustomizationLoaded()
   > n14 = Branch(Condition=CharacterItemDefinitionHasAnyMetaTag(Tags=Make Array!("[0]"=(TagName="Cosmetics.Variant.Property.Stage1"), "[1]":Gameplay Tag Structure=(TagName="Cosmetics.Variant.Property.Stage4")) -> (Array)))
   > Set hair_base_active(hair_base_active:bool=true) -> (Output_Get:bool)
   Set hair_base_active(execute=@n14.else, hair_base_active=false)
   Event BlueprintInitializeAnimation()
   > Set Mesh_To_Copy(Mesh_To_Copy:Skeletal Mesh Component Object Reference=GetOwnerSkeletalMeshComponent()) -> (Output_Get:Skeletal Mesh Component Object Reference)
   TryGetPawnOwner()
   Get bCanPlayCustomAnimations() -> (bCanPlayCustomAnimations:bool)
  ]}
]
variables [
 armSwingDelta_l {type=number}
 armSwingDelta_r {type=number}
 hair_base_active {type=boolean}
 head_rot_back {type=number}
 headSwingDelta {type=number}
 headTwist_l {type=number}
 headTwist_r {type=number}
 Mesh_To_Copy {type=SkeletalMeshComponent}
 Skydiving_Back_Alpha {type=number}
 Skydiving_Front_Alpha {type=number}
]
